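# Sound cue for the handshake whoosh-in foley of the Huzzah emote.
# A mixer layers a boosted boot-step foley with a randomly chosen one of three equally weighted whoosh-in variations, the whoosh passing through a short random delay.
# It is routed as emote music, using that sound class, concurrency rule and an audio-analysis submix send, with shared emote attenuation and play-when-silent virtualization.
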SoundCue /Game/Athena/Sounds/Emotes/Huzzah/EmoteFoley_Huzzah_Handshake_Whoosh_In_Cue
root=Mixer_0
nodes [
 {id=Delay_0, children [Random_1], properties {delayMax=0.2}}
 {id=Mixer_0, children [WavePlayer_0,Delay_0], properties {inputVolume [1.4,1]}}
 {id=Random_1, children [WavePlayer_1,WavePlayer_2,WavePlayer_3], properties {weights [1,1,1]}}
 {id=WavePlayer_0, properties {soundWaveAssetPtr=&/Emote_Huzzah_Owned_Boot_Step_Foley_04}}
 {id=WavePlayer_1, properties {soundWaveAssetPtr=&/Emote_Huzzah_Hand_Shake_Whoosh_In_01}}
 {id=WavePlayer_2, properties {soundWaveAssetPtr=&/Emote_Huzzah_Hand_Shake_Whoosh_In_02}}
 {id=WavePlayer_3, properties {soundWaveAssetPtr=&/Emote_Huzzah_Hand_Shake_Whoosh_In_03}}
]
properties {
 attenuationSettings=&Fort_Emotes_Athena_Att
 maxDistance=1600
 soundClassObject=&Fort_Music_Emote
 virtualizationMode=PlayWhenSilent
 volumeMultiplier=0.6
 concurrencyOverrides {maxCount=3}
 concurrencySet [/Game/Sounds/Concurrency_Rules/Athena_EmoteMusic_Conc]
 soundSubmixSends [{minSendLevel=0, soundSubmix=&Fortnite_AudioAnalysis_Music}]
}
